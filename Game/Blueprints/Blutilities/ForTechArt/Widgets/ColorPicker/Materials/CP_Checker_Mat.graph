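# UI material for the ColorPicker tech-art editor widget that builds a grayscale checkerboard pattern.
# Viewport UVs scaled by a Dimensions parameter sample a checker texture, whose red channel blends between dark and light gray, keeping the pattern screen-aligned.
# The graph also holds a Color-times-U gradient, a ScreenAlignedUVs call and a ViewSize node, none feeding the checker chain.
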Material /Game/Blueprints/Blutilities/ForTechArt/Widgets/ColorPicker/Materials/CP_Checker_Mat
graph {
 properties {materialDomain=MD_UI, pixelDepthOffsetMode=PDOM_Legacy, usedShadingModels=Default Lit}
 nodes [
  n1 = ScreenPosition()
  n2 = ScalarParameter() {defaultValue=20, parameterName=Dimensions}
  LinearInterpolate(Alpha=TextureSample(UVs=Multiply(A=@n1.ViewportUV, B=@n2)) {samplerType=SAMPLERTYPE_LinearGrayscale, texture=/Game/Global/Textures/Default/DefaultUI/Checkerboard_D}.R) {constA=0.2, constB=0.8}
  MaterialFunctionCall(Texture Dimensions=@n2) {materialFunction=&ScreenAlignedUVs}
  Multiply(A=VectorParameter() {parameterName=Color, defaultValue=<rgba 1 0 0 1>}.RGB, B=ComponentMask(None=TextureCoordinate()) {r=true})
  ViewSize()
 ]
}
