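# Tech-art color picker sub-widget that edits a color through Hue, Saturation, Value and Alpha sliders, spin boxes and a hex field.
# validateFinalColor rebuilds RGB from the sliders (saturation inverted), refreshes preview, slider tints, hex text and spin boxes, then broadcasts onColorUpdated.
# setColor and hex entry feed back through applyColorToHSVSliders, with flags gating broadcasts and spin-box feedback.
# Hex conversion is hand-built on an enum digit lookup, with debug prints.
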WidgetBlueprint /Game/Blueprints/Blutilities/ForTechArt/Widgets/ColorPicker/SubWidgets/TA_ColorSliderCollection_Widget
graphs [
 EventGraph {nodes [
   "Value Changed (slider_Val)"!() -> (OutputDelegate, then, value:double) {is=title K2Node_ComponentBoundEvent}
   > validateFinalColor()
   n1 = "On Value Changed (SpinBox_Value)"!() -> (OutputDelegate, then, InValue:float) {is=title K2Node_ComponentBoundEvent}
   > Branch(Condition=Get allowSpinBoxEvents() -> (allowSpinBoxEvents:bool))
   > setValue!(
    execute
    self:TA Color Slider Base Widget Object Reference=Get slider_Val() -> (slider_Val:TA Color Slider Base Widget Object Reference)
    InValue:double=@n1.InValue
    throwEvent:bool=true
   ) -> (then)
   n2 = "On Value Changed (SpinBox_Alpha)"!() -> (OutputDelegate, then, InValue:float) {is=title K2Node_ComponentBoundEvent}
   > Branch(Condition=Get allowSpinBoxEvents())
   > setValue(self=Get slider_Alpha() -> (slider_Alpha:TA Color Slider Base Widget Object Reference), InValue=@n2.InValue, throwEvent=true)
   "Value Changed (slider_Hue)"!() -> (OutputDelegate, then, value:double) {is=title K2Node_ComponentBoundEvent}
   > validateFinalColor()
   "Value Changed (slider_Sat)"!() -> (OutputDelegate, then, value:double) {is=title K2Node_ComponentBoundEvent}
   > validateFinalColor()
   "On Text Committed (EditableTextBox_75)"!() -> (OutputDelegate, then, Text:"Text (by ref)", CommitMethod:ETextCommit Enum=Default) {is=title K2Node_ComponentBoundEvent}
   > populateFromHex()
   "Value Changed (slider_Alpha)"!() -> (OutputDelegate, then, value:double) {is=title K2Node_ComponentBoundEvent}
   > validateFinalColor()
   n3 = "On Value Changed (SpinBox_Hue)"!() -> (OutputDelegate, then, InValue:float) {is=title K2Node_ComponentBoundEvent}
   > Branch(Condition=Get allowSpinBoxEvents())
   > setValue(
    self=Get slider_Hue() -> (slider_Hue:TA Color Slider Base Widget Object Reference)
    InValue=float / float(A=@n3.InValue, B=360) {is=title K2Node_CallFunction}
    throwEvent=true
   )
   n4 = "On Value Changed (SpinBox_Sat)"!() -> (OutputDelegate, then, InValue:float) {is=title K2Node_ComponentBoundEvent}
   > Branch(Condition=Get allowSpinBoxEvents())
   > setValue(self=Get slider_Sat() -> (slider_Sat:TA Color Slider Base Widget Object Reference), InValue=@n4.InValue, throwEvent=true)
   Event PreConstruct()
   > validateFinalColor()
  ]}
 GetHexValues {nodes [
   n1 = GetHexValues!() -> (then, InputColorChannel:double=~) {is=title K2Node_FunctionEntry}
   n2 = float / float(A=@n1, B=16)
   n3 = Floor(A=@n2) {is=title K2Node_CallFunction}
   n4 = Get Enum() -> (Enum:Array of HexVal_Enum Enums=NewEnumerator0)
   n5 = int * float!(A:int=16, B:double="% (Float)"!(A:double=@n2, B:double=1) -> (ReturnValue:double)) -> (ReturnValue:double)
   n6 = Floor(A=@n5)
   > Return Node!(
    execute
    1:String=Enum to String!(Enumerator:HexVal_Enum Enum="Get (a ref)"(Array:Array of HexVal_Enum Enums=@n4, Dimension 1=@n3) -> (Output:"HexVal_Enum Enum (by ref)")) -> (ReturnValue:String)
    2:String=Enum to String(Enumerator="Get (a ref)"(Array=@n4, Dimension 1=@n6))
   )
   Truncate(A=@n5)
  ]}
 GetRGBFromHex {nodes [
   n1 = GetRGBFromHex!() -> (then, HexValueSpace1:String, HexValueSpace2:String) {is=title K2Node_FunctionEntry}
   n2 = Get l_strings() -> (l_strings:Array of Strings)
   > Add(TargetArray=@n2, NewItem=@n1) {is=title K2Node_CallArrayFunction}
   > Add(TargetArray=@n2, NewItem=@n1.HexValueSpace2)
   > n3 = For Each Loop(Array=Get l_strings())
   n4 = Get l_values() -> (l_values:Array of Integers)
   n5 = "Get (a copy)"(Array:Array of Integers=@n4) -> (Output:int)
   Return Node!(
    execute=@n3.Completed
    ChannelValue:int=int + int(
     A=int * int(A:int=@n5, B:int=16) -> (ReturnValue:int=0) {is=title K2Node_CommutativeAssociativeBinaryOperator}
     B="Get (a copy)"(Array=@n4, Dimension 1=1)
    ) -> (ReturnValue=0) {is=title K2Node_CommutativeAssociativeBinaryOperator}
   )
   > PrintString!(
    execute
    InString=Append!(A=@n1, B=@n1.HexValueSpace2, C:String=------------------------------) -> (ReturnValue)
    bPrintToScreen
    bPrintToLog
    TextColor
    Duration
    Key
   ) -> (then)
   > n6 = For Each Loop with Break!(Exec:Exec, Array:Array of HexVal_Enum Enums=Get Enum() -> (Enum=NewEnumerator0), Break:Exec=@n8.then) -> (LoopBody:Exec, Array Element:HexVal_Enum Enum, Array Index:int, Completed:Exec)
   n7 = Enum to String(Enumerator=@n6)
   > PrintString(InString=Append(A=@n7, B=" =?= ", C=@n3))
   > Branch(Condition="Equal Exactly (String)"(A=@n7, B=@n3))
   > n8 = Add(TargetArray:Array of Integers=Get l_values(), NewItem:"Integer (by ref)"=@n6.Array Index)
   n9 = Branch(Condition="Equal Exactly (String)"(B=0))
   > n10 = Set V1(V1:int=0) -> (Output_Get:int=0)
   n11 = Branch(
    execute=[
     @n71.then
     @n44.then
     @n48.then
     @n52.then
     @n56.then
     @n60.then
     @n64.then
     @n68.then
     @n16.then
     @n20.then
     @n24.then
     @n10.then
     @n28.then
     @n32.then
     @n36.then
     @n40.then
    ]
    Condition="Equal Exactly (String)"(B=0)
   )
   > n12 = Set V2(V2:int=0) -> (Output_Get:int=0)
   > n13 = Set ValueFinal(ValueFinal:int=int + int(A=int * int(A=@n10, B=16) -> (ReturnValue=0), B=@n12) -> (ReturnValue=0)) -> (Output_Get:int=0)
   > Return Node(ChannelValue=@n13)
   n14 = Branch(execute=@n11.else, Condition="Equal Exactly (String)"(B=1))
   > n15 = Set V2(V2=1) -> (Output_Get=0)
   n16 = Set V1(execute=@n73.then, V1=1) -> (Output_Get=0)
   n17 = Set ValueFinal(execute=@n15.then, ValueFinal=int + int(A=int * int(A=@n16, B=16) -> (ReturnValue=0), B=@n15) -> (ReturnValue=0)) -> (Output_Get=0)
   > Return Node(ChannelValue=@n17)
   n18 = Branch(execute=@n14.else, Condition="Equal Exactly (String)"(B=2))
   > n19 = Set V2(V2=2) -> (Output_Get=0)
   n20 = Set V1(execute=@n74.then, V1=2) -> (Output_Get=0)
   n21 = Set ValueFinal(execute=@n19.then, ValueFinal=int + int(A=int * int(A=@n20, B=16) -> (ReturnValue=0), B=@n19) -> (ReturnValue=0)) -> (Output_Get=0)
   > Return Node(ChannelValue=@n21)
   n22 = Branch(execute=@n18.else, Condition="Equal Exactly (String)"(B=3))
   > n23 = Set V2(V2=3) -> (Output_Get=0)
   n24 = Set V1(execute=@n75.then, V1=3) -> (Output_Get=0)
   n25 = Set ValueFinal(execute=@n23.then, ValueFinal=int + int(A=int * int(A=@n24, B=16) -> (ReturnValue=0), B=@n23) -> (ReturnValue=0)) -> (Output_Get=0)
   > Return Node(ChannelValue=@n25)
   n26 = Branch(execute=@n22.else, Condition="Equal Exactly (String)"(B=4))
   > n27 = Set V2(V2=4) -> (Output_Get=0)
   n28 = Set V1(execute=@n76.then, V1=4) -> (Output_Get=0)
   n29 = Set ValueFinal(execute=@n27.then, ValueFinal=int + int(A=int * int(A=@n28, B=16) -> (ReturnValue=0), B=@n27) -> (ReturnValue=0)) -> (Output_Get=0)
   > Return Node(ChannelValue=@n29)
   n30 = Branch(execute=@n26.else, Condition="Equal Exactly (String)"(B=5))
   > n31 = Set V2(V2=5) -> (Output_Get=0)
   n32 = Set V1(execute=@n77.then, V1=5) -> (Output_Get=0)
   n33 = Set ValueFinal(execute=@n31.then, ValueFinal=int + int(A=int * int(A=@n32, B=16) -> (ReturnValue=0), B=@n31) -> (ReturnValue=0)) -> (Output_Get=0)
   > Return Node(ChannelValue=@n33)
   n34 = Branch(execute=@n30.else, Condition="Equal Exactly (String)"(B=6))
   > n35 = Set V2(V2=6) -> (Output_Get=0)
   n36 = Set V1(execute=@n78.then, V1=6) -> (Output_Get=0)
   n37 = Set ValueFinal(execute=@n35.then, ValueFinal=int + int(A=int * int(A=@n36, B=16) -> (ReturnValue=0), B=@n35) -> (ReturnValue=0)) -> (Output_Get=0)
   > Return Node(ChannelValue=@n37)
   n38 = Branch(execute=@n34.else, Condition="Equal Exactly (String)"(B=7))
   > n39 = Set V2(V2=7) -> (Output_Get=0)
   n40 = Set V1(execute=@n79.then, V1=7) -> (Output_Get=0)
   n41 = Set ValueFinal(execute=@n39.then, ValueFinal=int + int(A=int * int(A=@n40, B=16) -> (ReturnValue=0), B=@n39) -> (ReturnValue=0)) -> (Output_Get=0)
   > Return Node(ChannelValue=@n41)
   n42 = Branch(execute=@n38.else, Condition="Equal Exactly (String)"(B=8))
   > n43 = Set V2(V2=8) -> (Output_Get=0)
   n44 = Set V1(execute=@n80.then, V1=8) -> (Output_Get=0)
   n45 = Set ValueFinal(execute=@n43.then, ValueFinal=int + int(A=int * int(A=@n44, B=16) -> (ReturnValue=0), B=@n43) -> (ReturnValue=0)) -> (Output_Get=0)
   > Return Node(ChannelValue=@n45)
   n46 = Branch(execute=@n42.else, Condition="Equal Exactly (String)"(B=9))
   > n47 = Set V2(V2=9) -> (Output_Get=0)
   n48 = Set V1(execute=@n81.then, V1=9) -> (Output_Get=0)
   n49 = Set ValueFinal(execute=@n47.then, ValueFinal=int + int(A=int * int(A=@n48, B=16) -> (ReturnValue=0), B=@n47) -> (ReturnValue=0)) -> (Output_Get=0)
   > Return Node(ChannelValue=@n49)
   n50 = Branch(execute=@n46.else, Condition="Equal Exactly (String)"(B=A))
   > n51 = Set V2(V2=10) -> (Output_Get=0)
   n52 = Set V1(execute=@n82.then, V1=10) -> (Output_Get=0)
   n53 = Set ValueFinal(execute=@n51.then, ValueFinal=int + int(A=int * int(A=@n52, B=16) -> (ReturnValue=0), B=@n51) -> (ReturnValue=0)) -> (Output_Get=0)
   > Return Node(ChannelValue=@n53)
   n54 = Branch(execute=@n50.else, Condition="Equal Exactly (String)"(B=B))
   > n55 = Set V2(V2=11) -> (Output_Get=0)
   n56 = Set V1(execute=@n83.then, V1=11) -> (Output_Get=0)
   n57 = Set ValueFinal(execute=@n55.then, ValueFinal=int + int(A=int * int(A=@n56, B=16) -> (ReturnValue=0), B=@n55) -> (ReturnValue=0)) -> (Output_Get=0)
   > Return Node(ChannelValue=@n57)
   n58 = Branch(execute=@n54.else, Condition="Equal Exactly (String)"(B=C))
   > n59 = Set V2(V2=12) -> (Output_Get=0)
   n60 = Set V1(execute=@n84.then, V1=12) -> (Output_Get=0)
   n61 = Set ValueFinal(execute=@n59.then, ValueFinal=int + int(A=int * int(A=@n60, B=16) -> (ReturnValue=0), B=@n59) -> (ReturnValue=0)) -> (Output_Get=0)
   > Return Node(ChannelValue=@n61)
   n62 = Branch(execute=@n58.else, Condition="Equal Exactly (String)"(B=D))
   > n63 = Set V2(V2=13) -> (Output_Get=0)
   n64 = Set V1(execute=@n85.then, V1=13) -> (Output_Get=0)
   n65 = Set ValueFinal(execute=@n63.then, ValueFinal=int + int(A=int * int(A=@n64, B=16) -> (ReturnValue=0), B=@n63) -> (ReturnValue=0)) -> (Output_Get=0)
   > Return Node(ChannelValue=@n65)
   n66 = Branch(execute=@n62.else, Condition="Equal Exactly (String)"(B=E))
   > n67 = Set V2(V2=14) -> (Output_Get=0)
   n68 = Set V1(execute=@n86.then, V1=14) -> (Output_Get=0)
   n69 = Set ValueFinal(execute=@n67.then, ValueFinal=int + int(A=int * int(A=@n68, B=16) -> (ReturnValue=0), B=@n67) -> (ReturnValue=0)) -> (Output_Get=0)
   > Return Node(ChannelValue=@n69)
   Branch(execute=@n66.else, Condition="Equal Exactly (String)"(B=F))
   > n70 = Set V2(V2=15) -> (Output_Get=0)
   n71 = Set V1(execute=@n87.then, V1=15) -> (Output_Get=0)
   n72 = Set ValueFinal(execute=@n70.then, ValueFinal=int + int(A=int * int(A=@n71, B=16) -> (ReturnValue=0), B=@n70) -> (ReturnValue=0)) -> (Output_Get=0)
   > Return Node(ChannelValue=@n72)
   n73 = Branch(execute=@n9.else, Condition="Equal Exactly (String)"(B=1))
   n74 = Branch(execute=@n73.else, Condition="Equal Exactly (String)"(B=2))
   n75 = Branch(execute=@n74.else, Condition="Equal Exactly (String)"(B=3))
   n76 = Branch(execute=@n75.else, Condition="Equal Exactly (String)"(B=4))
   n77 = Branch(execute=@n76.else, Condition="Equal Exactly (String)"(B=5))
   n78 = Branch(execute=@n77.else, Condition="Equal Exactly (String)"(B=6))
   n79 = Branch(execute=@n78.else, Condition="Equal Exactly (String)"(B=7))
   n80 = Branch(execute=@n79.else, Condition="Equal Exactly (String)"(B=8))
   n81 = Branch(execute=@n80.else, Condition="Equal Exactly (String)"(B=9))
   n82 = Branch(execute=@n81.else, Condition="Equal Exactly (String)"(B=A))
   n83 = Branch(execute=@n82.else, Condition="Equal Exactly (String)"(B=B))
   n84 = Branch(execute=@n83.else, Condition="Equal Exactly (String)"(B=C))
   n85 = Branch(execute=@n84.else, Condition="Equal Exactly (String)"(B=D))
   n86 = Branch(execute=@n85.else, Condition="Equal Exactly (String)"(B=E))
   n87 = Branch(execute=@n86.else, Condition="Equal Exactly (String)"(B=F))
   Set V1(V1=0) -> (Output_Get=0)
  ]}
 applyColorToHSVSliders {nodes [
   n1 = applyColorToHSVSliders!() -> (then, InColor:Linear Color Structure) {is=title K2Node_FunctionEntry}
   > setValue(
    self=Get slider_Hue()
    InValue=float / float(A=RGB to HSV!(InColor:Linear Color Structure=@n1) -> (H:float, S:float, V:float, A:float), B=360)
    throwEvent=false
   )
   > setValue(self=Get slider_Sat(), InValue=float - float(B=RGB to HSV(InColor=@n1).S) {is=title K2Node_CallFunction}, throwEvent=false)
   n2 = RGB to HSV(InColor=@n1)
   > setValue(self=Get slider_Val(), InValue=@n2.V, throwEvent=false)
   > setValue(self=Get slider_Alpha(), InValue=@n2.A, throwEvent=false)
   > validateFinalColor()
  ]}
 getCurrentColorAsHex {nodes [
   getCurrentColorAsHex() {is=title K2Node_FunctionEntry}
   > Set l_finalHex(l_finalHex:String="##") -> (Output_Get:String)
   n1 = Break Linear Color!(LinearColor:"Linear Color Structure (by ref)"=Get currentColor() -> (currentColor:Linear Color Structure)) -> (R:float=~, G:float=~, B:float=~, A:float=~) {is=title K2Node_BreakStruct}
   > n2 = GetHexValues!(execute, self, InputColorChannel:double=float * float(A=@n1, B=255)) -> (then, 1:String, 2:String) {is=title K2Node_CallFunction}
   > n3 = GetHexValues(InputColorChannel=float * float(A=@n1.G, B=255))
   > n4 = GetHexValues(InputColorChannel=float * float(A=@n1.B, B=255))
   > n5 = GetHexValues(InputColorChannel=float * float(A=@n1.A, B=255))
   > n6 = Set l_finalHex(l_finalHex=Append!(
     A=Get l_finalHex() -> (l_finalHex:String)
     B=@n2
     C=@n2.2
     D:String=@n3
     E:String=@n3.2
     F:String=@n4
     G:String=@n4.2
     H:String=@n5
     I:String=@n5.2
    ) -> (ReturnValue))
   > Return Node!(execute, hex:String=@n6)
  ]}
 getHexAsColor {nodes [
   getHexAsColor() {is=title K2Node_FunctionEntry}
   n1 = "To String (Text)"(InText="GetText (Text Box)"!(self:Text Box Object Reference=Get EditableTextBox_75() -> (EditableTextBox_75:Text Box Object Reference)) -> (ReturnValue:Text))
   > Set l_stringArray(l_stringArray:Array of Strings=GetCharacterArrayFromString!(SourceString:String=RightChop!(SourceString:String=@n1, Count:int=1) -> (ReturnValue:String)) -> (ReturnValue:Array of Strings)) -> (Output_Get:Array of Strings)
   n2 = Get l_stringArray() -> (l_stringArray:Array of Strings)
   > n3 = GetRGBFromHex!(
    execute
    self
    HexValueSpace1:String="Get (a copy)"(Array:Array of Strings=@n2) -> (Output:String)
    HexValueSpace2:String="Get (a copy)"(Array=@n2, Dimension 1=1)
   ) -> (then, ChannelValue:int=0) {is=title K2Node_CallFunction}
   n4 = Get l_stringArray()
   > n5 = GetRGBFromHex(HexValueSpace1="Get (a copy)"(Array=@n4, Dimension 1=2), HexValueSpace2="Get (a copy)"(Array=@n4, Dimension 1=3)) -> (ChannelValue=0)
   n6 = Get l_stringArray()
   > n7 = GetRGBFromHex(HexValueSpace1="Get (a copy)"(Array=@n6, Dimension 1=4), HexValueSpace2="Get (a copy)"(Array=@n6, Dimension 1=5)) -> (ChannelValue=0)
   n8 = Get l_stringArray()
   > n9 = GetRGBFromHex(HexValueSpace1="Get (a copy)"(Array=@n8, Dimension 1=6), HexValueSpace2="Get (a copy)"(Array=@n8, Dimension 1=7)) -> (ChannelValue=0)
   n10 = Make Literal Float(Value=255)
   n11 = "To Float (Integer)"(InInt=@n9)
   n12 = "To Float (Integer)"(InInt=@n3)
   n13 = "To Float (Integer)"(InInt=@n5)
   n14 = "To Float (Integer)"(InInt=@n7)
   > Return Node!(
    execute
    valid:bool=true
    color:Linear Color Structure=MakeColor!(
     R:float=float / float(A=@n12, B=@n10)
     G:float=float / float(A=@n13, B=@n10)
     B:float=float / float(A=@n14, B=@n10)
     A:float=float / float(A=@n11, B=@n10)
    ) -> (ReturnValue:Linear Color Structure)
   )
  ]}
 onColorUpdated {nodes [onColorUpdated!() -> (then, color:Linear Color Structure) {is=title K2Node_FunctionEntry}]}
 populateFromHex {nodes [
   populateFromHex() {is=title K2Node_FunctionEntry}
   > n1 = getHexAsColor!(execute, self) -> (then, valid:bool, color:Linear Color Structure) {is=title K2Node_CallFunction}
   > n2 = Branch(Condition=@n1)
   > Set l_color(l_color:Linear Color Structure=@n1.color) -> (Output_Get:Linear Color Structure)
   > applyColorToHSVSliders!(execute, self, InColor:Linear Color Structure=Get l_color() -> (l_color:Linear Color Structure)) -> (then) {is=title K2Node_CallFunction}
   validateHexText(execute=@n2.else)
  ]}
 setColor {nodes [
   n1 = setColor!() -> (then, color:Linear Color Structure, throwEvents:bool=~) {is=title K2Node_FunctionEntry}
   > Set allowEvents(allowEvents:bool=@n1.throwEvents) -> (Output_Get:bool)
   > applyColorToHSVSliders(InColor=@n1)
   > Set allowEvents(allowEvents=true)
  ]}
 validateFinalColor {nodes [
   validateFinalColor() {is=title K2Node_FunctionEntry}
   > n1 = Sequence!(execute) -> (then_0, then_1, then_2:Exec, then_3:Exec, then_4:Exec)
   n2 = Get slider_Hue()
   > Set l_Hue(l_Hue:double=float * float(A=getValue!(self:TA Color Slider Base Widget Object Reference=@n2) -> (value:double), B=360)) -> (Output_Get:double)
   n3 = Get slider_Sat()
   > Set l_Sat(l_Sat:double=float - float(B=getValue(self=@n3))) -> (Output_Get:double)
   > Set l_Val(l_Val:double=getValue(self=Get slider_Val())) -> (Output_Get:double)
   > Set l_Alpha(l_Alpha:double=getValue(self=Get slider_Alpha())) -> (Output_Get:double)
   Set currentColor(
    execute=@n1.then_1
    currentColor:Linear Color Structure=HSV to RGB!(
     H:float=Get l_Hue() -> (l_Hue:double)
     S:float=Get l_Sat() -> (l_Sat:double)
     V:float=Get l_Val() -> (l_Val:double)
     A:float=Get l_Alpha() -> (l_Alpha:double)
    ) -> (ReturnValue:Linear Color Structure)
   ) -> (Output_Get:Linear Color Structure)
   > SetColorAndOpacity(self:Image Object Reference=Get Image_75() -> (Image_75:Image Object Reference), InColorAndOpacity:Linear Color Structure=Get currentColor())
   > SetColor!(
    execute
    self:TA Color Slider Base Widget Object Reference=Get slider_Sat()
    color:Linear Color Structure=HSV to RGB(H=Get l_Hue(), S=1, V=1, A=1)
   ) -> (then)
   > SetColor(self=Get slider_Val(), color=HSV to RGB(H=Get l_Hue(), S=Get l_Sat(), V=1, A=1))
   validateHexText(execute=@n1.then_2)
   Set allowSpinBoxEvents(execute=@n1.then_3, allowSpinBoxEvents:bool=false) -> (Output_Get:bool)
   n4 = Get slider_Hue()
   > SetValue!(
    execute
    self:Spin Box Object Reference=Get SpinBox_Hue() -> (SpinBox_Hue:Spin Box Object Reference)
    NewValue:float=float * float(A=getValue(self=@n4), B=360)
   ) -> (then)
   > SetValue(self=Get SpinBox_Sat() -> (SpinBox_Sat:Spin Box Object Reference), NewValue=getValue(self=Get slider_Sat()))
   > SetValue(self=Get SpinBox_Value() -> (SpinBox_Value:Spin Box Object Reference), NewValue=getValue(self=Get slider_Val()))
   > SetValue(self=Get SpinBox_Alpha() -> (SpinBox_Alpha:Spin Box Object Reference), NewValue=getValue(self=Get slider_Alpha()))
   > Set allowSpinBoxEvents(allowSpinBoxEvents=true)
   Branch(execute=@n1.then_4, Condition=Get allowEvents() -> (allowEvents:bool))
   > Call On Color Updated!(execute:Exec, self:TA Color Slider Collection Widget Object Reference, color:Linear Color Structure=Get currentColor()) -> (then:Exec) {is=title K2Node_CallDelegate}
  ]}
 validateHexText {nodes [
   validateHexText() {is=title K2Node_FunctionEntry}
   > n1 = getCurrentColorAsHex!(execute, self) -> (then, hex:String) {is=title K2Node_CallFunction}
   > "SetText (Text Box)"!(execute, self:Text Box Object Reference=Get EditableTextBox_75(), InText:Text="To Text (String)"(InString=@n1)) -> (then)
  ]}
]
variables [
 allowEvents {type=boolean}
 allowSpinBoxEvents {type=boolean, default=true}
 currentColor {type=LinearColor, default=<rgba 0.3608 0.4157 0.3647 1>}
 Enum {type="array<HexVal_Enum>", default [0,1,2,3,4,5,6,7,8,9,A,B,C,D,E,F]}
 onColorUpdated {}
]
widgets {
 info {inheritedWidgetCount=0, parentClass=&UserWidget, rootWidgetClass=&CanvasPanel, widgetCount=27}
 widgets [
  CanvasPanel_0 {widgetClassPath=&CanvasPanel, widget=CanvasPanel_0}
  HorizontalBox_0 {
   widgetClassPath=&HorizontalBox
   parent=CanvasPanel_0
   widget=HorizontalBox_0
   slot {
    class=CanvasPanelSlot
    properties {bAutoSize=true, layoutData {alignment=<0 0>, anchors {maximum=<1 0>, minimum=<0 0>}, offsets=<ltrb 0 0 0 30>}}
   }
  }
  VerticalBox_0 {
   widgetClassPath=&VerticalBox
   parent=HorizontalBox_0
   widget=VerticalBox_0
   slot {class=HorizontalBoxSlot, properties {size {sizeRule=Fill, value=1}}}
  }
  HorizontalBox_2 {widgetClassPath=&HorizontalBox, parent=VerticalBox_0, widget=HorizontalBox_2, slot {class=VerticalBoxSlot}}
  VerticalBox_1 {
   widgetClassPath=&VerticalBox
   parent=HorizontalBox_2
   widget=VerticalBox_1
   slot {class=HorizontalBoxSlot, properties {size {sizeRule=Fill, value=1}}}
  }
  HorizontalBox_3 {widgetClassPath=&HorizontalBox, parent=VerticalBox_1, widget=HorizontalBox_3, slot {class=VerticalBoxSlot}}
  CommonTextBlock_3 {
   widgetClassPath=&CommonTextBlock
   parent=HorizontalBox_3
   widget=CommonTextBlock_3
   properties {
    justification=Right
    minDesiredWidth=40
    text=Hue:
    font {
     bForceMonospaced=false
     bMaterialIsStencil=false
     fontMaterial=null
     fontObject=&NotoSans
     letterSpacing=0
     monospacedWidth=1
     size=10
     skewAmount=0
     typefaceFontName=NotoSans-Regular
     outlineSettings {
      bApplyOutlineToDropShadows=false
      bMiteredCorners=false
      bSeparateFillAlpha=false
      outlineMaterial=null
      outlineSize=0
      outlineColor=<rgba 0 0 0 1>
     }
    }
    shadowOffset=<0 0>
   }
   slot {class=HorizontalBoxSlot, properties {padding=<ltrb 0 0 3 0>}}
  }
  slider_Hue {
   widgetClassPath=&/TA_ColorSliderBase_Widget
   bIsVariable=true
   parent=HorizontalBox_3
   widget=slider_Hue
   properties {materialForSliderBG=/Game/Blueprints/Blutilities/ForTechArt/Widgets/ColorPicker/Materials/CP_Hue_Mat}
   slot {class=HorizontalBoxSlot, properties {verticalAlignment=VAlign_Center, size {sizeRule=Fill, value=1}}}
  }
  SpinBox_Hue {
   widgetClassPath=/Script/UMG.SpinBox
   bIsVariable=true
   parent=HorizontalBox_3
   widget=SpinBox_Hue
   properties {
    bAlwaysUsesDeltaSnap=true
    bOverride_MaxSliderValue=true
    bOverride_MaxValue=true
    bOverride_MinSliderValue=true
    bOverride_MinValue=true
    delta=1
    maxFractionalDigits=0
    maxSliderValue=360
    maxValue=360
    minDesiredWidth=40
    minFractionalDigits=0
    font {
     bForceMonospaced=false
     bMaterialIsStencil=false
     fontMaterial=null
     fontObject=&Roboto
     letterSpacing=0
     monospacedWidth=1
     size=10
     skewAmount=0
     typefaceFontName=Bold
     outlineSettings {
      bApplyOutlineToDropShadows=false
      bMiteredCorners=false
      bSeparateFillAlpha=false
      outlineMaterial=null
      outlineSize=0
      outlineColor=<rgba 0 0 0 1>
     }
    }
    widgetStyle {
     activeBackgroundBrush {
      bIsDynamicallyLoaded=false
      drawAs=NoDrawType
      imageType=NoImage
      mirroring=NoMirror
      resourceName=None
      resourceObject=null
      tiling=NoTile
      imageSize=<0 0>
      margin=<ltrb 0 0 0 0>
      outlineSettings {
       bUseBrushTransparency=false
       roundingType=HalfHeightRadius
       width=0
       color {colorUseRule=UseColor_Specified, specifiedColor=<rgba 0 0 0 0>}
       cornerRadii=<0 0 0 0>
      }
      tintColor {colorUseRule=UseColor_Specified, specifiedColor=<rgba 1 1 1 1>}
      uVRegion {bIsValid=false, max=<0 0>, min=<0 0>}
     }
     activeFillBrush {
      bIsDynamicallyLoaded=false
      drawAs=Box
      imageType=FullColor
      mirroring=NoMirror
      resourceName=../../../Engine/Content/Slate/Common/Spinbox_Fill.png
      resourceObject=null
      tiling=NoTile
      imageSize=<0 0>
      margin=<ltrb 0.25 0.25 0.5 0.25>
      outlineSettings {
       bUseBrushTransparency=false
       roundingType=HalfHeightRadius
       width=0
       color {colorUseRule=UseColor_Specified, specifiedColor=<rgba 0 0 0 0>}
       cornerRadii=<0 0 0 0>
      }
      tintColor {colorUseRule=UseColor_Specified, specifiedColor=<rgba 1 1 1 1>}
      uVRegion {bIsValid=false, max=<0 0>, min=<0 0>}
     }
     arrowsImage {
      bIsDynamicallyLoaded=false
      drawAs=Image
      imageType=FullColor
      mirroring=NoMirror
      resourceName=../../../Engine/Content/Slate/Common/SpinArrows.png
      resourceObject=null
      tiling=NoTile
      imageSize=<12 12>
      margin=<ltrb 0 0 0 0>
      outlineSettings {
       bUseBrushTransparency=false
       roundingType=HalfHeightRadius
       width=0
       color {colorUseRule=UseColor_Specified, specifiedColor=<rgba 0 0 0 0>}
       cornerRadii=<0 0 0 0>
      }
      tintColor {colorUseRule=UseColor_Specified, specifiedColor=<rgba 1 1 1 1>}
      uVRegion {bIsValid=false, max=<0 0>, min=<0 0>}
     }
     backgroundBrush {
      bIsDynamicallyLoaded=false
      drawAs=Box
      imageType=FullColor
      mirroring=NoMirror
      resourceName=../../../Engine/Content/Slate/Common/Spinbox.png
      resourceObject=null
      tiling=NoTile
      imageSize=<0 0>
      margin=<ltrb 0.25 0.25 0.25 0.25>
      outlineSettings {
       bUseBrushTransparency=false
       roundingType=HalfHeightRadius
       width=0
       color {colorUseRule=UseColor_Specified, specifiedColor=<rgba 0 0 0 0>}
       cornerRadii=<0 0 0 0>
      }
      tintColor {colorUseRule=UseColor_Specified, specifiedColor=<rgba 1 1 1 1>}
      uVRegion {bIsValid=false, max=<0 0>, min=<0 0>}
     }
     foregroundColor {colorUseRule=UseColor_Specified, specifiedColor=<rgba 0.4729 0.4729 0.4729 1>}
     hoveredBackgroundBrush {
      bIsDynamicallyLoaded=false
      drawAs=Box
      imageType=FullColor
      mirroring=NoMirror
      resourceName=../../../Engine/Content/Slate/Common/Spinbox_Hovered.png
      resourceObject=null
      tiling=NoTile
      imageSize=<0 0>
      margin=<ltrb 0.25 0.25 0.25 0.25>
      outlineSettings {
       bUseBrushTransparency=false
       roundingType=HalfHeightRadius
       width=0
       color {colorUseRule=UseColor_Specified, specifiedColor=<rgba 0 0 0 0>}
       cornerRadii=<0 0 0 0>
      }
      tintColor {colorUseRule=UseColor_Specified, specifiedColor=<rgba 1 1 1 1>}
      uVRegion {bIsValid=false, max=<0 0>, min=<0 0>}
     }
     hoveredFillBrush {
      bIsDynamicallyLoaded=false
      drawAs=NoDrawType
      imageType=NoImage
      mirroring=NoMirror
      resourceName=None
      resourceObject=null
      tiling=NoTile
      imageSize=<0 0>
      margin=<ltrb 0 0 0 0>
      outlineSettings {
       bUseBrushTransparency=false
       roundingType=HalfHeightRadius
       width=0
       color {colorUseRule=UseColor_Specified, specifiedColor=<rgba 0 0 0 0>}
       cornerRadii=<0 0 0 0>
      }
      tintColor {colorUseRule=UseColor_Specified, specifiedColor=<rgba 1 1 1 1>}
      uVRegion {bIsValid=false, max=<0 0>, min=<0 0>}
     }
     inactiveFillBrush {
      bIsDynamicallyLoaded=false
      drawAs=Box
      imageType=FullColor
      mirroring=NoMirror
      resourceName=../../../Engine/Content/Slate/Common/Spinbox_Fill_Hovered.png
      resourceObject=null
      tiling=NoTile
      imageSize=<0 0>
      margin=<ltrb 0.25 0.25 0.25 0.25>
      outlineSettings {
       bUseBrushTransparency=false
       roundingType=HalfHeightRadius
       width=0
       color {colorUseRule=UseColor_Specified, specifiedColor=<rgba 0 0 0 0>}
       cornerRadii=<0 0 0 0>
      }
      tintColor {colorUseRule=UseColor_Specified, specifiedColor=<rgba 1 1 1 1>}
      uVRegion {bIsValid=false, max=<0 0>, min=<0 0>}
     }
     insetPadding=<ltrb 0 0 0 0>
     textPadding=<ltrb 1 1 1 1>
    }
   }
   slot {class=HorizontalBoxSlot, properties {verticalAlignment=VAlign_Center, padding=<ltrb 3 0 0 0>}}
  }
  HorizontalBox_5 {
   widgetClassPath=&HorizontalBox
   parent=VerticalBox_1
   widget=HorizontalBox_5
   slot {class=VerticalBoxSlot, properties {padding=<ltrb 0 5 0 0>}}
  }
  CommonTextBlock_2 {
   widgetClassPath=&CommonTextBlock
   parent=HorizontalBox_5
   widget=CommonTextBlock_2
   properties {
    justification=Right
    minDesiredWidth=40
    text=Sat:
    font {
     bForceMonospaced=false
     bMaterialIsStencil=false
     fontMaterial=null
     fontObject=&NotoSans
     letterSpacing=0
     monospacedWidth=1
     size=10
     skewAmount=0
     typefaceFontName=NotoSans-Regular
     outlineSettings {
      bApplyOutlineToDropShadows=false
      bMiteredCorners=false
      bSeparateFillAlpha=false
      outlineMaterial=null
      outlineSize=0
      outlineColor=<rgba 0 0 0 1>
     }
    }
    shadowOffset=<0 0>
   }
   slot {class=HorizontalBoxSlot, properties {padding=<ltrb 0 0 3 0>}}
  }
  slider_Sat {
   widgetClassPath=&TA_ColorSliderBase_Widget
   bIsVariable=true
   parent=HorizontalBox_5
   widget=slider_Sat
   properties {materialForSliderBG=/Game/Blueprints/Blutilities/ForTechArt/Widgets/ColorPicker/Materials/CP_Saturation_Mat}
   slot {class=HorizontalBoxSlot, properties {verticalAlignment=VAlign_Center, size {sizeRule=Fill, value=1}}}
  }
  SpinBox_Sat {
   widgetClassPath=&SpinBox
   bIsVariable=true
   parent=HorizontalBox_5
   widget=SpinBox_Sat
   properties {
    bAlwaysUsesDeltaSnap=true
    bOverride_MaxSliderValue=true
    bOverride_MaxValue=true
    bOverride_MinSliderValue=true
    bOverride_MinValue=true
    delta=0.01
    maxFractionalDigits=2
    maxSliderValue=1
    maxValue=1
    minDesiredWidth=40
    minFractionalDigits=2
    font {
     bForceMonospaced=false
     bMaterialIsStencil=false
     fontMaterial=null
     fontObject=&Roboto
     letterSpacing=0
     monospacedWidth=1
     size=10
     skewAmount=0
     typefaceFontName=Bold
     outlineSettings {
      bApplyOutlineToDropShadows=false
      bMiteredCorners=false
      bSeparateFillAlpha=false
      outlineMaterial=null
      outlineSize=0
      outlineColor=<rgba 0 0 0 1>
     }
    }
    widgetStyle {
     activeBackgroundBrush {
      bIsDynamicallyLoaded=false
      drawAs=NoDrawType
      imageType=NoImage
      mirroring=NoMirror
      resourceName=None
      resourceObject=null
      tiling=NoTile
      imageSize=<0 0>
      margin=<ltrb 0 0 0 0>
      outlineSettings {
       bUseBrushTransparency=false
       roundingType=HalfHeightRadius
       width=0
       color {colorUseRule=UseColor_Specified, specifiedColor=<rgba 0 0 0 0>}
       cornerRadii=<0 0 0 0>
      }
      tintColor {colorUseRule=UseColor_Specified, specifiedColor=<rgba 1 1 1 1>}
      uVRegion {bIsValid=false, max=<0 0>, min=<0 0>}
     }
     activeFillBrush {
      bIsDynamicallyLoaded=false
      drawAs=Box
      imageType=FullColor
      mirroring=NoMirror
      resourceName=../../../Engine/Content/Slate/Common/Spinbox_Fill.png
      resourceObject=null
      tiling=NoTile
      imageSize=<0 0>
      margin=<ltrb 0.25 0.25 0.5 0.25>
      outlineSettings {
       bUseBrushTransparency=false
       roundingType=HalfHeightRadius
       width=0
       color {colorUseRule=UseColor_Specified, specifiedColor=<rgba 0 0 0 0>}
       cornerRadii=<0 0 0 0>
      }
      tintColor {colorUseRule=UseColor_Specified, specifiedColor=<rgba 1 1 1 1>}
      uVRegion {bIsValid=false, max=<0 0>, min=<0 0>}
     }
     arrowsImage {
      bIsDynamicallyLoaded=false
      drawAs=Image
      imageType=FullColor
      mirroring=NoMirror
      resourceName=../../../Engine/Content/Slate/Common/SpinArrows.png
      resourceObject=null
      tiling=NoTile
      imageSize=<12 12>
      margin=<ltrb 0 0 0 0>
      outlineSettings {
       bUseBrushTransparency=false
       roundingType=HalfHeightRadius
       width=0
       color {colorUseRule=UseColor_Specified, specifiedColor=<rgba 0 0 0 0>}
       cornerRadii=<0 0 0 0>
      }
      tintColor {colorUseRule=UseColor_Specified, specifiedColor=<rgba 1 1 1 1>}
      uVRegion {bIsValid=false, max=<0 0>, min=<0 0>}
     }
     backgroundBrush {
      bIsDynamicallyLoaded=false
      drawAs=Box
      imageType=FullColor
      mirroring=NoMirror
      resourceName=../../../Engine/Content/Slate/Common/Spinbox.png
      resourceObject=null
      tiling=NoTile
      imageSize=<0 0>
      margin=<ltrb 0.25 0.25 0.25 0.25>
      outlineSettings {
       bUseBrushTransparency=false
       roundingType=HalfHeightRadius
       width=0
       color {colorUseRule=UseColor_Specified, specifiedColor=<rgba 0 0 0 0>}
       cornerRadii=<0 0 0 0>
      }
      tintColor {colorUseRule=UseColor_Specified, specifiedColor=<rgba 1 1 1 1>}
      uVRegion {bIsValid=false, max=<0 0>, min=<0 0>}
     }
     foregroundColor {colorUseRule=UseColor_Specified, specifiedColor=<rgba 0.4729 0.4729 0.4729 1>}
     hoveredBackgroundBrush {
      bIsDynamicallyLoaded=false
      drawAs=Box
      imageType=FullColor
      mirroring=NoMirror
      resourceName=../../../Engine/Content/Slate/Common/Spinbox_Hovered.png
      resourceObject=null
      tiling=NoTile
      imageSize=<0 0>
      margin=<ltrb 0.25 0.25 0.25 0.25>
      outlineSettings {
       bUseBrushTransparency=false
       roundingType=HalfHeightRadius
       width=0
       color {colorUseRule=UseColor_Specified, specifiedColor=<rgba 0 0 0 0>}
       cornerRadii=<0 0 0 0>
      }
      tintColor {colorUseRule=UseColor_Specified, specifiedColor=<rgba 1 1 1 1>}
      uVRegion {bIsValid=false, max=<0 0>, min=<0 0>}
     }
     hoveredFillBrush {
      bIsDynamicallyLoaded=false
      drawAs=NoDrawType
      imageType=NoImage
      mirroring=NoMirror
      resourceName=None
      resourceObject=null
      tiling=NoTile
      imageSize=<0 0>
      margin=<ltrb 0 0 0 0>
      outlineSettings {
       bUseBrushTransparency=false
       roundingType=HalfHeightRadius
       width=0
       color {colorUseRule=UseColor_Specified, specifiedColor=<rgba 0 0 0 0>}
       cornerRadii=<0 0 0 0>
      }
      tintColor {colorUseRule=UseColor_Specified, specifiedColor=<rgba 1 1 1 1>}
      uVRegion {bIsValid=false, max=<0 0>, min=<0 0>}
     }
     inactiveFillBrush {
      bIsDynamicallyLoaded=false
      drawAs=Box
      imageType=FullColor
      mirroring=NoMirror
      resourceName=../../../Engine/Content/Slate/Common/Spinbox_Fill_Hovered.png
      resourceObject=null
      tiling=NoTile
      imageSize=<0 0>
      margin=<ltrb 0.25 0.25 0.25 0.25>
      outlineSettings {
       bUseBrushTransparency=false
       roundingType=HalfHeightRadius
       width=0
       color {colorUseRule=UseColor_Specified, specifiedColor=<rgba 0 0 0 0>}
       cornerRadii=<0 0 0 0>
      }
      tintColor {colorUseRule=UseColor_Specified, specifiedColor=<rgba 1 1 1 1>}
      uVRegion {bIsValid=false, max=<0 0>, min=<0 0>}
     }
     insetPadding=<ltrb 0 0 0 0>
     textPadding=<ltrb 1 1 1 1>
    }
   }
   slot {class=HorizontalBoxSlot, properties {verticalAlignment=VAlign_Center, padding=<ltrb 3 0 0 0>}}
  }
  HorizontalBox_7 {
   widgetClassPath=&HorizontalBox
   parent=VerticalBox_1
   widget=HorizontalBox_7
   slot {class=VerticalBoxSlot, properties {padding=<ltrb 0 5 0 0>}}
  }
  CommonTextBlock_1 {
   widgetClassPath=&CommonTextBlock
   parent=HorizontalBox_7
   widget=CommonTextBlock_1
   properties {
    justification=Right
    minDesiredWidth=40
    text=Value:
    font {
     bForceMonospaced=false
     bMaterialIsStencil=false
     fontMaterial=null
     fontObject=&NotoSans
     letterSpacing=0
     monospacedWidth=1
     size=10
     skewAmount=0
     typefaceFontName=NotoSans-Regular
     outlineSettings {
      bApplyOutlineToDropShadows=false
      bMiteredCorners=false
      bSeparateFillAlpha=false
      outlineMaterial=null
      outlineSize=0
      outlineColor=<rgba 0 0 0 1>
     }
    }
    shadowOffset=<0 0>
   }
   slot {class=HorizontalBoxSlot, properties {padding=<ltrb 0 0 3 0>}}
  }
  slider_Val {
   widgetClassPath=&TA_ColorSliderBase_Widget
   bIsVariable=true
   parent=HorizontalBox_7
   widget=slider_Val
   properties {materialForSliderBG=/Game/Blueprints/Blutilities/ForTechArt/Widgets/ColorPicker/Materials/CP_Value_Mat, startValue=1}
   slot {class=HorizontalBoxSlot, properties {verticalAlignment=VAlign_Center, size {sizeRule=Fill, value=1}}}
  }
  SpinBox_Value {
   widgetClassPath=&SpinBox
   bIsVariable=true
   parent=HorizontalBox_7
   widget=SpinBox_Value
   properties {
    bAlwaysUsesDeltaSnap=true
    bOverride_MaxSliderValue=true
    bOverride_MaxValue=true
    bOverride_MinSliderValue=true
    bOverride_MinValue=true
    delta=0.01
    maxFractionalDigits=2
    maxSliderValue=1
    maxValue=1
    minDesiredWidth=40
    minFractionalDigits=2
    font {
     bForceMonospaced=false
     bMaterialIsStencil=false
     fontMaterial=null
     fontObject=&Roboto
     letterSpacing=0
     monospacedWidth=1
     size=10
     skewAmount=0
     typefaceFontName=Bold
     outlineSettings {
      bApplyOutlineToDropShadows=false
      bMiteredCorners=false
      bSeparateFillAlpha=false
      outlineMaterial=null
      outlineSize=0
      outlineColor=<rgba 0 0 0 1>
     }
    }
    widgetStyle {
     activeBackgroundBrush {
      bIsDynamicallyLoaded=false
      drawAs=NoDrawType
      imageType=NoImage
      mirroring=NoMirror
      resourceName=None
      resourceObject=null
      tiling=NoTile
      imageSize=<0 0>
      margin=<ltrb 0 0 0 0>
      outlineSettings {
       bUseBrushTransparency=false
       roundingType=HalfHeightRadius
       width=0
       color {colorUseRule=UseColor_Specified, specifiedColor=<rgba 0 0 0 0>}
       cornerRadii=<0 0 0 0>
      }
      tintColor {colorUseRule=UseColor_Specified, specifiedColor=<rgba 1 1 1 1>}
      uVRegion {bIsValid=false, max=<0 0>, min=<0 0>}
     }
     activeFillBrush {
      bIsDynamicallyLoaded=false
      drawAs=Box
      imageType=FullColor
      mirroring=NoMirror
      resourceName=../../../Engine/Content/Slate/Common/Spinbox_Fill.png
      resourceObject=null
      tiling=NoTile
      imageSize=<0 0>
      margin=<ltrb 0.25 0.25 0.5 0.25>
      outlineSettings {
       bUseBrushTransparency=false
       roundingType=HalfHeightRadius
       width=0
       color {colorUseRule=UseColor_Specified, specifiedColor=<rgba 0 0 0 0>}
       cornerRadii=<0 0 0 0>
      }
      tintColor {colorUseRule=UseColor_Specified, specifiedColor=<rgba 1 1 1 1>}
      uVRegion {bIsValid=false, max=<0 0>, min=<0 0>}
     }
     arrowsImage {
      bIsDynamicallyLoaded=false
      drawAs=Image
      imageType=FullColor
      mirroring=NoMirror
      resourceName=../../../Engine/Content/Slate/Common/SpinArrows.png
      resourceObject=null
      tiling=NoTile
      imageSize=<12 12>
      margin=<ltrb 0 0 0 0>
      outlineSettings {
       bUseBrushTransparency=false
       roundingType=HalfHeightRadius
       width=0
       color {colorUseRule=UseColor_Specified, specifiedColor=<rgba 0 0 0 0>}
       cornerRadii=<0 0 0 0>
      }
      tintColor {colorUseRule=UseColor_Specified, specifiedColor=<rgba 1 1 1 1>}
      uVRegion {bIsValid=false, max=<0 0>, min=<0 0>}
     }
     backgroundBrush {
      bIsDynamicallyLoaded=false
      drawAs=Box
      imageType=FullColor
      mirroring=NoMirror
      resourceName=../../../Engine/Content/Slate/Common/Spinbox.png
      resourceObject=null
      tiling=NoTile
      imageSize=<0 0>
      margin=<ltrb 0.25 0.25 0.25 0.25>
      outlineSettings {
       bUseBrushTransparency=false
       roundingType=HalfHeightRadius
       width=0
       color {colorUseRule=UseColor_Specified, specifiedColor=<rgba 0 0 0 0>}
       cornerRadii=<0 0 0 0>
      }
      tintColor {colorUseRule=UseColor_Specified, specifiedColor=<rgba 1 1 1 1>}
      uVRegion {bIsValid=false, max=<0 0>, min=<0 0>}
     }
     foregroundColor {colorUseRule=UseColor_Specified, specifiedColor=<rgba 0.4729 0.4729 0.4729 1>}
     hoveredBackgroundBrush {
      bIsDynamicallyLoaded=false
      drawAs=Box
      imageType=FullColor
      mirroring=NoMirror
      resourceName=../../../Engine/Content/Slate/Common/Spinbox_Hovered.png
      resourceObject=null
      tiling=NoTile
      imageSize=<0 0>
      margin=<ltrb 0.25 0.25 0.25 0.25>
      outlineSettings {
       bUseBrushTransparency=false
       roundingType=HalfHeightRadius
       width=0
       color {colorUseRule=UseColor_Specified, specifiedColor=<rgba 0 0 0 0>}
       cornerRadii=<0 0 0 0>
      }
      tintColor {colorUseRule=UseColor_Specified, specifiedColor=<rgba 1 1 1 1>}
      uVRegion {bIsValid=false, max=<0 0>, min=<0 0>}
     }
     hoveredFillBrush {
      bIsDynamicallyLoaded=false
      drawAs=NoDrawType
      imageType=NoImage
      mirroring=NoMirror
      resourceName=None
      resourceObject=null
      tiling=NoTile
      imageSize=<0 0>
      margin=<ltrb 0 0 0 0>
      outlineSettings {
       bUseBrushTransparency=false
       roundingType=HalfHeightRadius
       width=0
       color {colorUseRule=UseColor_Specified, specifiedColor=<rgba 0 0 0 0>}
       cornerRadii=<0 0 0 0>
      }
      tintColor {colorUseRule=UseColor_Specified, specifiedColor=<rgba 1 1 1 1>}
      uVRegion {bIsValid=false, max=<0 0>, min=<0 0>}
     }
     inactiveFillBrush {
      bIsDynamicallyLoaded=false
      drawAs=Box
      imageType=FullColor
      mirroring=NoMirror
      resourceName=../../../Engine/Content/Slate/Common/Spinbox_Fill_Hovered.png
      resourceObject=null
      tiling=NoTile
      imageSize=<0 0>
      margin=<ltrb 0.25 0.25 0.25 0.25>
      outlineSettings {
       bUseBrushTransparency=false
       roundingType=HalfHeightRadius
       width=0
       color {colorUseRule=UseColor_Specified, specifiedColor=<rgba 0 0 0 0>}
       cornerRadii=<0 0 0 0>
      }
      tintColor {colorUseRule=UseColor_Specified, specifiedColor=<rgba 1 1 1 1>}
      uVRegion {bIsValid=false, max=<0 0>, min=<0 0>}
     }
     insetPadding=<ltrb 0 0 0 0>
     textPadding=<ltrb 1 1 1 1>
    }
   }
   slot {class=HorizontalBoxSlot, properties {verticalAlignment=VAlign_Center, padding=<ltrb 3 0 0 0>}}
  }
  HorizontalBox_8 {
   widgetClassPath=&HorizontalBox
   parent=VerticalBox_1
   widget=HorizontalBox_8
   slot {class=VerticalBoxSlot, properties {padding=<ltrb 0 5 0 0>}}
  }
  CommonTextBlock {
   widgetClassPath=&CommonTextBlock
   parent=HorizontalBox_8
   widget=CommonTextBlock
   properties {
    minDesiredWidth=40
    text=Alpha:
    font {
     bForceMonospaced=false
     bMaterialIsStencil=false
     fontMaterial=null
     fontObject=&NotoSans
     letterSpacing=0
     monospacedWidth=1
     size=10
     skewAmount=0
     typefaceFontName=NotoSans-Regular
     outlineSettings {
      bApplyOutlineToDropShadows=false
      bMiteredCorners=false
      bSeparateFillAlpha=false
      outlineMaterial=null
      outlineSize=0
      outlineColor=<rgba 0 0 0 1>
     }
    }
    shadowOffset=<0 0>
   }
   slot {class=HorizontalBoxSlot, properties {padding=<ltrb 0 0 3 0>}}
  }
  slider_Alpha {
   widgetClassPath=&TA_ColorSliderBase_Widget
   bIsVariable=true
   parent=HorizontalBox_8
   widget=slider_Alpha
   properties {materialForSliderBG=/Game/Blueprints/Blutilities/ForTechArt/Widgets/ColorPicker/Materials/CP_Alpha_Mat, startValue=1}
   slot {class=HorizontalBoxSlot, properties {verticalAlignment=VAlign_Center, size {sizeRule=Fill, value=1}}}
  }
  SpinBox_Alpha {
   widgetClassPath=&SpinBox
   bIsVariable=true
   parent=HorizontalBox_8
   widget=SpinBox_Alpha
   properties {
    bAlwaysUsesDeltaSnap=true
    bOverride_MaxSliderValue=true
    bOverride_MaxValue=true
    bOverride_MinSliderValue=true
    bOverride_MinValue=true
    delta=0.01
    maxFractionalDigits=2
    maxSliderValue=1
    maxValue=1
    minDesiredWidth=40
    minFractionalDigits=2
    font {
     bForceMonospaced=false
     bMaterialIsStencil=false
     fontMaterial=null
     fontObject=&Roboto
     letterSpacing=0
     monospacedWidth=1
     size=10
     skewAmount=0
     typefaceFontName=Bold
     outlineSettings {
      bApplyOutlineToDropShadows=false
      bMiteredCorners=false
      bSeparateFillAlpha=false
      outlineMaterial=null
      outlineSize=0
      outlineColor=<rgba 0 0 0 1>
     }
    }
    widgetStyle {
     activeBackgroundBrush {
      bIsDynamicallyLoaded=false
      drawAs=NoDrawType
      imageType=NoImage
      mirroring=NoMirror
      resourceName=None
      resourceObject=null
      tiling=NoTile
      imageSize=<0 0>
      margin=<ltrb 0 0 0 0>
      outlineSettings {
       bUseBrushTransparency=false
       roundingType=HalfHeightRadius
       width=0
       color {colorUseRule=UseColor_Specified, specifiedColor=<rgba 0 0 0 0>}
       cornerRadii=<0 0 0 0>
      }
      tintColor {colorUseRule=UseColor_Specified, specifiedColor=<rgba 1 1 1 1>}
      uVRegion {bIsValid=false, max=<0 0>, min=<0 0>}
     }
     activeFillBrush {
      bIsDynamicallyLoaded=false
      drawAs=Box
      imageType=FullColor
      mirroring=NoMirror
      resourceName=../../../Engine/Content/Slate/Common/Spinbox_Fill.png
      resourceObject=null
      tiling=NoTile
      imageSize=<0 0>
      margin=<ltrb 0.25 0.25 0.5 0.25>
      outlineSettings {
       bUseBrushTransparency=false
       roundingType=HalfHeightRadius
       width=0
       color {colorUseRule=UseColor_Specified, specifiedColor=<rgba 0 0 0 0>}
       cornerRadii=<0 0 0 0>
      }
      tintColor {colorUseRule=UseColor_Specified, specifiedColor=<rgba 1 1 1 1>}
      uVRegion {bIsValid=false, max=<0 0>, min=<0 0>}
     }
     arrowsImage {
      bIsDynamicallyLoaded=false
      drawAs=Image
      imageType=FullColor
      mirroring=NoMirror
      resourceName=../../../Engine/Content/Slate/Common/SpinArrows.png
      resourceObject=null
      tiling=NoTile
      imageSize=<12 12>
      margin=<ltrb 0 0 0 0>
      outlineSettings {
       bUseBrushTransparency=false
       roundingType=HalfHeightRadius
       width=0
       color {colorUseRule=UseColor_Specified, specifiedColor=<rgba 0 0 0 0>}
       cornerRadii=<0 0 0 0>
      }
      tintColor {colorUseRule=UseColor_Specified, specifiedColor=<rgba 1 1 1 1>}
      uVRegion {bIsValid=false, max=<0 0>, min=<0 0>}
     }
     backgroundBrush {
      bIsDynamicallyLoaded=false
      drawAs=Box
      imageType=FullColor
      mirroring=NoMirror
      resourceName=../../../Engine/Content/Slate/Common/Spinbox.png
      resourceObject=null
      tiling=NoTile
      imageSize=<0 0>
      margin=<ltrb 0.25 0.25 0.25 0.25>
      outlineSettings {
       bUseBrushTransparency=false
       roundingType=HalfHeightRadius
       width=0
       color {colorUseRule=UseColor_Specified, specifiedColor=<rgba 0 0 0 0>}
       cornerRadii=<0 0 0 0>
      }
      tintColor {colorUseRule=UseColor_Specified, specifiedColor=<rgba 1 1 1 1>}
      uVRegion {bIsValid=false, max=<0 0>, min=<0 0>}
     }
     foregroundColor {colorUseRule=UseColor_Specified, specifiedColor=<rgba 0.4729 0.4729 0.4729 1>}
     hoveredBackgroundBrush {
      bIsDynamicallyLoaded=false
      drawAs=Box
      imageType=FullColor
      mirroring=NoMirror
      resourceName=../../../Engine/Content/Slate/Common/Spinbox_Hovered.png
      resourceObject=null
      tiling=NoTile
      imageSize=<0 0>
      margin=<ltrb 0.25 0.25 0.25 0.25>
      outlineSettings {
       bUseBrushTransparency=false
       roundingType=HalfHeightRadius
       width=0
       color {colorUseRule=UseColor_Specified, specifiedColor=<rgba 0 0 0 0>}
       cornerRadii=<0 0 0 0>
      }
      tintColor {colorUseRule=UseColor_Specified, specifiedColor=<rgba 1 1 1 1>}
      uVRegion {bIsValid=false, max=<0 0>, min=<0 0>}
     }
     hoveredFillBrush {
      bIsDynamicallyLoaded=false
      drawAs=NoDrawType
      imageType=NoImage
      mirroring=NoMirror
      resourceName=None
      resourceObject=null
      tiling=NoTile
      imageSize=<0 0>
      margin=<ltrb 0 0 0 0>
      outlineSettings {
       bUseBrushTransparency=false
       roundingType=HalfHeightRadius
       width=0
       color {colorUseRule=UseColor_Specified, specifiedColor=<rgba 0 0 0 0>}
       cornerRadii=<0 0 0 0>
      }
      tintColor {colorUseRule=UseColor_Specified, specifiedColor=<rgba 1 1 1 1>}
      uVRegion {bIsValid=false, max=<0 0>, min=<0 0>}
     }
     inactiveFillBrush {
      bIsDynamicallyLoaded=false
      drawAs=Box
      imageType=FullColor
      mirroring=NoMirror
      resourceName=../../../Engine/Content/Slate/Common/Spinbox_Fill_Hovered.png
      resourceObject=null
      tiling=NoTile
      imageSize=<0 0>
      margin=<ltrb 0.25 0.25 0.25 0.25>
      outlineSettings {
       bUseBrushTransparency=false
       roundingType=HalfHeightRadius
       width=0
       color {colorUseRule=UseColor_Specified, specifiedColor=<rgba 0 0 0 0>}
       cornerRadii=<0 0 0 0>
      }
      tintColor {colorUseRule=UseColor_Specified, specifiedColor=<rgba 1 1 1 1>}
      uVRegion {bIsValid=false, max=<0 0>, min=<0 0>}
     }
     insetPadding=<ltrb 0 0 0 0>
     textPadding=<ltrb 1 1 1 1>
    }
   }
   slot {class=HorizontalBoxSlot, properties {verticalAlignment=VAlign_Center, padding=<ltrb 3 0 0 0>}}
  }
  HorizontalBox_1 {
   widgetClassPath=&HorizontalBox
   parent=VerticalBox_0
   widget=HorizontalBox_1
   slot {class=VerticalBoxSlot, properties {padding=<ltrb 0 5 0 0>}}
  }
  CommonTextBlock_97 {
   widgetClassPath=&CommonTextBlock
   parent=HorizontalBox_1
   widget=CommonTextBlock_97
   properties {
    text="Hex Linear: "
    font {
     bForceMonospaced=false
     bMaterialIsStencil=false
     fontMaterial=null
     fontObject=&NotoSans
     letterSpacing=0
     monospacedWidth=1
     size=10
     skewAmount=0
     typefaceFontName=NotoSans-Regular
     outlineSettings {
      bApplyOutlineToDropShadows=false
      bMiteredCorners=false
      bSeparateFillAlpha=false
      outlineMaterial=null
      outlineSize=0
      outlineColor=<rgba 0 0 0 1>
     }
    }
    shadowOffset=<0 0>
   }
   slot {class=HorizontalBoxSlot}
  }
  EditableTextBox_75 {
   widgetClassPath=&EditableTextBox
   bIsVariable=true
   parent=HorizontalBox_1
   widget=EditableTextBox_75
   properties {
    justification=Center
    minimumDesiredWidth=60
    selectAllTextOnCommit=true
    selectAllTextWhenFocused=true
    widgetStyle {
     backgroundColor {colorUseRule=UseColor_Specified, specifiedColor=<rgba 1 1 1 1>}
     backgroundImageFocused {
      bIsDynamicallyLoaded=false
      drawAs=Box
      imageType=FullColor
      mirroring=NoMirror
      resourceName=../../../Engine/Content/Slate/Common/TextBox_Hovered.png
      resourceObject=null
      tiling=NoTile
      imageSize=<0 0>
      margin=<ltrb 0.25 0.25 0.25 0.25>
      outlineSettings {
       bUseBrushTransparency=false
       roundingType=HalfHeightRadius
       width=0
       color {colorUseRule=UseColor_Specified, specifiedColor=<rgba 0 0 0 0>}
       cornerRadii=<0 0 0 0>
      }
      tintColor {colorUseRule=UseColor_Specified, specifiedColor=<rgba 1 1 1 1>}
      uVRegion {bIsValid=false, max=<0 0>, min=<0 0>}
     }
     backgroundImageHovered {
      bIsDynamicallyLoaded=false
      drawAs=Box
      imageType=FullColor
      mirroring=NoMirror
      resourceName=../../../Engine/Content/Slate/Common/TextBox_Hovered.png
      resourceObject=null
      tiling=NoTile
      imageSize=<0 0>
      margin=<ltrb 0.25 0.25 0.25 0.25>
      outlineSettings {
       bUseBrushTransparency=false
       roundingType=HalfHeightRadius
       width=0
       color {colorUseRule=UseColor_Specified, specifiedColor=<rgba 0 0 0 0>}
       cornerRadii=<0 0 0 0>
      }
      tintColor {colorUseRule=UseColor_Specified, specifiedColor=<rgba 1 1 1 1>}
      uVRegion {bIsValid=false, max=<0 0>, min=<0 0>}
     }
     backgroundImageNormal {
      bIsDynamicallyLoaded=false
      drawAs=Box
      imageType=FullColor
      mirroring=NoMirror
      resourceName=../../../Engine/Content/Slate/Common/TextBox.png
      resourceObject=null
      tiling=NoTile
      imageSize=<0 0>
      margin=<ltrb 0.25 0.25 0.25 0.25>
      outlineSettings {
       bUseBrushTransparency=false
       roundingType=HalfHeightRadius
       width=0
       color {colorUseRule=UseColor_Specified, specifiedColor=<rgba 0 0 0 0>}
       cornerRadii=<0 0 0 0>
      }
      tintColor {colorUseRule=UseColor_Specified, specifiedColor=<rgba 1 1 1 1>}
      uVRegion {bIsValid=false, max=<0 0>, min=<0 0>}
     }
     backgroundImageReadOnly {
      bIsDynamicallyLoaded=false
      drawAs=Box
      imageType=FullColor
      mirroring=NoMirror
      resourceName=../../../Engine/Content/Slate/Common/TextBox_ReadOnly.png
      resourceObject=null
      tiling=NoTile
      imageSize=<0 0>
      margin=<ltrb 0.25 0.25 0.25 0.25>
      outlineSettings {
       bUseBrushTransparency=false
       roundingType=HalfHeightRadius
       width=0
       color {colorUseRule=UseColor_Specified, specifiedColor=<rgba 0 0 0 0>}
       cornerRadii=<0 0 0 0>
      }
      tintColor {colorUseRule=UseColor_Specified, specifiedColor=<rgba 1 1 1 1>}
      uVRegion {bIsValid=false, max=<0 0>, min=<0 0>}
     }
     focusedForegroundColor {colorUseRule=UseColor_Foreground, specifiedColor=<rgba 1 0 1 1>}
     foregroundColor {colorUseRule=UseColor_Foreground, specifiedColor=<rgba 1 0 1 1>}
     hScrollBarPadding=<ltrb 0 0 0 0>
     padding=<ltrb 4 2 4 2>
     readOnlyForegroundColor {colorUseRule=UseColor_Foreground, specifiedColor=<rgba 1 0 1 1>}
     scrollBarStyle {
      thickness=16
      draggedThumbImage {
       bIsDynamicallyLoaded=false
       drawAs=Box
       imageType=FullColor
       mirroring=NoMirror
       resourceName=../../../Engine/Content/Slate/Common/Scrollbar_Thumb.png
       resourceObject=null
       tiling=NoTile
       imageSize=<0 0>
       margin=<ltrb 0.25 0.25 0.25 0.25>
       outlineSettings {
        bUseBrushTransparency=false
        roundingType=HalfHeightRadius
        width=0
        color {colorUseRule=UseColor_Specified, specifiedColor=<rgba 0 0 0 0>}
        cornerRadii=<0 0 0 0>
       }
       tintColor {colorUseRule=UseColor_Specified, specifiedColor=<rgba 1 1 1 1>}
       uVRegion {bIsValid=false, max=<0 0>, min=<0 0>}
      }
      horizontalBackgroundImage {
       bIsDynamicallyLoaded=false
       drawAs=NoDrawType
       imageType=NoImage
       mirroring=NoMirror
       resourceName=None
       resourceObject=null
       tiling=NoTile
       imageSize=<0 0>
       margin=<ltrb 0 0 0 0>
       outlineSettings {
        bUseBrushTransparency=false
        roundingType=HalfHeightRadius
        width=0
        color {colorUseRule=UseColor_Specified, specifiedColor=<rgba 0 0 0 0>}
        cornerRadii=<0 0 0 0>
       }
       tintColor {colorUseRule=UseColor_Specified, specifiedColor=<rgba 1 1 1 1>}
       uVRegion {bIsValid=false, max=<0 0>, min=<0 0>}
      }
      horizontalBottomSlotImage {
       bIsDynamicallyLoaded=false
       drawAs=Image
       imageType=FullColor
       mirroring=NoMirror
       resourceName=../../../Engine/Content/Slate/Common/Scrollbar_Background_Horizontal.png
       resourceObject=null
       tiling=NoTile
       imageSize=<8 8>
       margin=<ltrb 0 0 0 0>
       outlineSettings {
        bUseBrushTransparency=false
        roundingType=HalfHeightRadius
        width=0
        color {colorUseRule=UseColor_Specified, specifiedColor=<rgba 0 0 0 0>}
        cornerRadii=<0 0 0 0>
       }
       tintColor {colorUseRule=UseColor_Specified, specifiedColor=<rgba 1 1 1 1>}
       uVRegion {bIsValid=false, max=<0 0>, min=<0 0>}
      }
      horizontalTopSlotImage {
       bIsDynamicallyLoaded=false
       drawAs=Image
       imageType=FullColor
       mirroring=NoMirror
       resourceName=../../../Engine/Content/Slate/Common/Scrollbar_Background_Horizontal.png
       resourceObject=null
       tiling=NoTile
       imageSize=<8 8>
       margin=<ltrb 0 0 0 0>
       outlineSettings {
        bUseBrushTransparency=false
        roundingType=HalfHeightRadius
        width=0
        color {colorUseRule=UseColor_Specified, specifiedColor=<rgba 0 0 0 0>}
        cornerRadii=<0 0 0 0>
       }
       tintColor {colorUseRule=UseColor_Specified, specifiedColor=<rgba 1 1 1 1>}
       uVRegion {bIsValid=false, max=<0 0>, min=<0 0>}
      }
      hoveredThumbImage {
       bIsDynamicallyLoaded=false
       drawAs=Box
       imageType=FullColor
       mirroring=NoMirror
       resourceName=../../../Engine/Content/Slate/Common/Scrollbar_Thumb.png
       resourceObject=null
       tiling=NoTile
       imageSize=<0 0>
       margin=<ltrb 0.25 0.25 0.25 0.25>
       outlineSettings {
        bUseBrushTransparency=false
        roundingType=HalfHeightRadius
        width=0
        color {colorUseRule=UseColor_Specified, specifiedColor=<rgba 0 0 0 0>}
        cornerRadii=<0 0 0 0>
       }
       tintColor {colorUseRule=UseColor_Specified, specifiedColor=<rgba 1 1 1 1>}
       uVRegion {bIsValid=false, max=<0 0>, min=<0 0>}
      }
      normalThumbImage {
       bIsDynamicallyLoaded=false
       drawAs=Box
       imageType=FullColor
       mirroring=NoMirror
       resourceName=../../../Engine/Content/Slate/Common/Scrollbar_Thumb.png
       resourceObject=null
       tiling=NoTile
       imageSize=<0 0>
       margin=<ltrb 0.25 0.25 0.25 0.25>
       outlineSettings {
        bUseBrushTransparency=false
        roundingType=HalfHeightRadius
        width=0
        color {colorUseRule=UseColor_Specified, specifiedColor=<rgba 0 0 0 0>}
        cornerRadii=<0 0 0 0>
       }
       tintColor {colorUseRule=UseColor_Specified, specifiedColor=<rgba 1 1 1 1>}
       uVRegion {bIsValid=false, max=<0 0>, min=<0 0>}
      }
      verticalBackgroundImage {
       bIsDynamicallyLoaded=false
       drawAs=NoDrawType
       imageType=NoImage
       mirroring=NoMirror
       resourceName=None
       resourceObject=null
       tiling=NoTile
       imageSize=<0 0>
       margin=<ltrb 0 0 0 0>
       outlineSettings {
        bUseBrushTransparency=false
        roundingType=HalfHeightRadius
        width=0
        color {colorUseRule=UseColor_Specified, specifiedColor=<rgba 0 0 0 0>}
        cornerRadii=<0 0 0 0>
       }
       tintColor {colorUseRule=UseColor_Specified, specifiedColor=<rgba 1 1 1 1>}
       uVRegion {bIsValid=false, max=<0 0>, min=<0 0>}
      }
      verticalBottomSlotImage {
       bIsDynamicallyLoaded=false
       drawAs=Image
       imageType=FullColor
       mirroring=NoMirror
       resourceName=../../../Engine/Content/Slate/Common/Scrollbar_Background_Vertical.png
       resourceObject=null
       tiling=NoTile
       imageSize=<8 8>
       margin=<ltrb 0 0 0 0>
       outlineSettings {
        bUseBrushTransparency=false
        roundingType=HalfHeightRadius
        width=0
        color {colorUseRule=UseColor_Specified, specifiedColor=<rgba 0 0 0 0>}
        cornerRadii=<0 0 0 0>
       }
       tintColor {colorUseRule=UseColor_Specified, specifiedColor=<rgba 1 1 1 1>}
       uVRegion {bIsValid=false, max=<0 0>, min=<0 0>}
      }
      verticalTopSlotImage {
       bIsDynamicallyLoaded=false
       drawAs=Image
       imageType=FullColor
       mirroring=NoMirror
       resourceName=../../../Engine/Content/Slate/Common/Scrollbar_Background_Vertical.png
       resourceObject=null
       tiling=NoTile
       imageSize=<8 8>
       margin=<ltrb 0 0 0 0>
       outlineSettings {
        bUseBrushTransparency=false
        roundingType=HalfHeightRadius
        width=0
        color {colorUseRule=UseColor_Specified, specifiedColor=<rgba 0 0 0 0>}
        cornerRadii=<0 0 0 0>
       }
       tintColor {colorUseRule=UseColor_Specified, specifiedColor=<rgba 1 1 1 1>}
       uVRegion {bIsValid=false, max=<0 0>, min=<0 0>}
      }
     }
     textStyle {
      overflowPolicy=Clip
      transformPolicy=None
      colorAndOpacity {colorUseRule=UseColor_Foreground, specifiedColor=<rgba 1 0 1 1>}
      font {
       bForceMonospaced=false
       bMaterialIsStencil=false
       fontMaterial=null
       fontObject=null
       letterSpacing=0
       monospacedWidth=1
       size=9
       skewAmount=0
       typefaceFontName=Regular
       outlineSettings {
        bApplyOutlineToDropShadows=false
        bMiteredCorners=false
        bSeparateFillAlpha=false
        outlineMaterial=null
        outlineSize=0
        outlineColor=<rgba 0 0 0 1>
       }
      }
      highlightColor {colorUseRule=UseColor_Specified, specifiedColor=<rgba 0.02 0.3 0 1>}
      highlightShape {
       bIsDynamicallyLoaded=false
       drawAs=Box
       imageType=FullColor
       mirroring=NoMirror
       resourceName=../../../Engine/Content/Slate/Common/TextBlockHighlightShape.png
       resourceObject=null
       tiling=NoTile
       imageSize=<0 0>
       margin=<ltrb 0.375 0.375 0.375 0.375>
       outlineSettings {
        bUseBrushTransparency=false
        roundingType=HalfHeightRadius
        width=0
        color {colorUseRule=UseColor_Specified, specifiedColor=<rgba 0 0 0 0>}
        cornerRadii=<0 0 0 0>
       }
       tintColor {colorUseRule=UseColor_Specified, specifiedColor=<rgba 1 1 1 1>}
       uVRegion {bIsValid=false, max=<0 0>, min=<0 0>}
      }
      selectedBackgroundColor {colorUseRule=UseColor_Foreground, specifiedColor=<rgba 1 0 1 1>}
      shadowColorAndOpacity=<rgba 0 0 0 1>
      shadowOffset=<0 0>
      strikeBrush {
       bIsDynamicallyLoaded=false
       drawAs=Image
       imageType=NoImage
       mirroring=NoMirror
       resourceName=None
       resourceObject=null
       tiling=NoTile
       imageSize=<32 32>
       margin=<ltrb 0 0 0 0>
       outlineSettings {
        bUseBrushTransparency=false
        roundingType=HalfHeightRadius
        width=0
        color {colorUseRule=UseColor_Specified, specifiedColor=<rgba 0 0 0 0>}
        cornerRadii=<0 0 0 0>
       }
       tintColor {colorUseRule=UseColor_Specified, specifiedColor=<rgba 1 1 1 1>}
       uVRegion {bIsValid=false, max=<0 0>, min=<0 0>}
      }
      underlineBrush {
       bIsDynamicallyLoaded=false
       drawAs=Image
       imageType=NoImage
       mirroring=NoMirror
       resourceName=None
       resourceObject=null
       tiling=NoTile
       imageSize=<32 32>
       margin=<ltrb 0 0 0 0>
       outlineSettings {
        bUseBrushTransparency=false
        roundingType=HalfHeightRadius
        width=0
        color {colorUseRule=UseColor_Specified, specifiedColor=<rgba 0 0 0 0>}
        cornerRadii=<0 0 0 0>
       }
       tintColor {colorUseRule=UseColor_Specified, specifiedColor=<rgba 1 1 1 1>}
       uVRegion {bIsValid=false, max=<0 0>, min=<0 0>}
      }
     }
     vScrollBarPadding=<ltrb 0 0 0 0>
    }
   }
   slot {class=HorizontalBoxSlot}
  }
  Border_0 {
   widgetClassPath=&Border
   parent=HorizontalBox_0
   widget=Border_0
   properties {brushColor=<rgba 0.0208 0.0208 0.0208 1>, padding=<ltrb 4 4 4 4>}
   slot {class=HorizontalBoxSlot}
  }
  Border_1 {
   widgetClassPath=&Border
   parent=Border_0
   widget=Border_1
   properties {
    background {
     bIsDynamicallyLoaded=false
     drawAs=Image
     imageType=NoImage
     mirroring=NoMirror
     resourceName=None
     resourceObject=/Game/Blueprints/Blutilities/ForTechArt/Widgets/ColorPicker/Materials/CP_Checker_Mat
     tiling=Both
     imageSize=<32 64>
     margin=<ltrb 0 0 0 0>
     outlineSettings {
      bUseBrushTransparency=false
      roundingType=HalfHeightRadius
      width=0
      color {colorUseRule=UseColor_Specified, specifiedColor=<rgba 0 0 0 0>}
      cornerRadii=<0 0 0 0>
     }
     tintColor {colorUseRule=UseColor_Specified, specifiedColor=<rgba 1 1 1 1>}
     uVRegion {bIsValid=false, max=<0 0>, min=<0 0>}
    }
    padding=<ltrb 0 0 0 0>
   }
   slot {class=BorderSlot, properties {padding=<ltrb 4 4 4 4>}}
  }
  Image_75 {
   widgetClassPath=&Image
   bIsVariable=true
   parent=Border_1
   widget=Image_75
   slot {class=BorderSlot, properties {padding=<ltrb 0 0 0 0>}}
  }
 ]
}
